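# Physics asset for the cape on the Dark Minion back bling, named for cloth but built from rigid bodies on bones.
# Center, left and right three-bone chains, mostly capsules, each link to a small MSTR root body.
# All three roots are rotation-locked to a large pack capsule.
# Chain joints carry small angle settings, widest on swing2, and the right chain also locks one twist and one swing1 axis.
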PhysicsAsset /BRCosmetics/Accessories/FORT_Backpacks/Backpack_M_MED_Dark_Minion/Meshes/M_MED_Dark_Minion_Cloth_Physics
bodies {
 dyn_cape_01 [{shapeType=Sphere, radius=1.75, shapeName=dyn_cape_01_sphere, center=<4.5461 0 0.8018>}]
 dyn_cape_01_MSTR [{shapeType=Capsule, length=0.505, radius=0.505, rotation=<rot 0 -90 0>}]
 dyn_cape_01_l [{shapeType=Capsule, length=4.8575, radius=1.1555, center=<5.035 -0.2328 0.3268>, rotation=<rot 0 -94.7017 -84.5829>}]
 dyn_cape_01_l_MSTR [{shapeType=Capsule, length=0.505, radius=0.505, rotation=<rot 0 -90 0>}]
 dyn_cape_01_r [{shapeType=Capsule, length=5.3248, radius=1.5, center=<4.544 0.8138 0.4197>, rotation=<rot 90 -179.9901 -179.9902>}]
 dyn_cape_01_r_MSTR [{shapeType=Capsule, length=0.505, radius=0.505, rotation=<rot 0 -90 0>}]
 dyn_cape_02 [{shapeType=Capsule, length=4.166, radius=1.9528, center=<3.633 0.0961 1.1362>, rotation=<rot 2.2807 -92.0695 -88.2544>}]
 dyn_cape_02_l [{shapeType=Capsule, length=4.1352, radius=1.5143, center=<3.6265 -0.0076 -0.3271>, rotation=<rot -2.3736 -87.6192 -87.1683>}]
 dyn_cape_02_r [{shapeType=Capsule, length=4.1532, radius=2.4879, center=<4.2747 0.2192 0.1629>, rotation=<rot -90 -144.7356 144.7357>}]
 dyn_cape_03 [{shapeType=Capsule, length=4.0988, radius=1.75, center=<1.6137 -0.1315 0.5617>, rotation=<rot 0.0581 86.1654 -82.0663>}]
 dyn_cape_03_l [{shapeType=Capsule, length=0.505, radius=1.005, rotation=<rot 0 -90 0>}]
 dyn_cape_03_r [{shapeType=Capsule, length=0, radius=0.75, center=<0.843 0 -0.3613>, rotation=<rot 0 -89.9998 90.0006>}]
 pack [{shapeType=Capsule, length=19.6643, radius=10.0577, center=<0.4691 8.953 -20.8662>, rotation=<rot -1.9635 -89.9901 -0.9924>}]
}
constraints [
 {bone1Name=dyn_cape_01_r, bone2Name=dyn_cape_01_r_MSTR, swing1LimitDegrees=2, swing2LimitDegrees=10, twistLimitDegrees=2, twistMotion=Locked}
 {bone1Name=dyn_cape_02_r, bone2Name=dyn_cape_01_r, swing1LimitDegrees=5, swing1Motion=Locked, swing2LimitDegrees=20, twistLimitDegrees=2}
 {bone1Name=dyn_cape_03_r, bone2Name=dyn_cape_02_r, swing1LimitDegrees=5, swing2LimitDegrees=20, twistLimitDegrees=2}
 {bone1Name=dyn_cape_03_l, bone2Name=dyn_cape_02_l, swing1LimitDegrees=2, swing2LimitDegrees=20, twistLimitDegrees=2}
 {bone1Name=dyn_cape_02_l, bone2Name=dyn_cape_01_l, swing1LimitDegrees=2, swing2LimitDegrees=20, twistLimitDegrees=2}
 {bone1Name=dyn_cape_01_l, bone2Name=dyn_cape_01_l_MSTR, swing1LimitDegrees=2, swing2LimitDegrees=10, twistLimitDegrees=2}
 {bone1Name=dyn_cape_03, bone2Name=dyn_cape_02, swing1LimitDegrees=10, swing2LimitDegrees=30, twistLimitDegrees=2}
 {bone1Name=dyn_cape_02, bone2Name=dyn_cape_01, swing1LimitDegrees=10, swing2LimitDegrees=30, twistLimitDegrees=2}
 {bone1Name=dyn_cape_01, bone2Name=dyn_cape_01_MSTR, swing1LimitDegrees=2, swing2LimitDegrees=10, twistLimitDegrees=0}
 {
  bone1Name=dyn_cape_01_MSTR
  bone2Name=pack
  swing1LimitDegrees=45
  swing1Motion=Locked
  swing2LimitDegrees=45
  swing2Motion=Locked
  twistLimitDegrees=45
  twistMotion=Locked
 }
 {
  bone1Name=dyn_cape_01_l_MSTR
  bone2Name=pack
  swing1LimitDegrees=45
  swing1Motion=Locked
  swing2LimitDegrees=45
  swing2Motion=Locked
  twistLimitDegrees=45
  twistMotion=Locked
 }
 {
  bone1Name=dyn_cape_01_r_MSTR
  bone2Name=pack
  swing1LimitDegrees=45
  swing1Motion=Locked
  swing2LimitDegrees=45
  swing2Motion=Locked
  twistLimitDegrees=45
  twistMotion=Locked
 }
]
