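# Physics asset for the Dark Minion backpack cosmetic, setting up physical secondary motion for the feathers on the pack.
# A large capsule on the pack bone anchors sixteen feather chains of two or three small bodies, all capsules except one sphere.
# Each chain links segment to segment back to the pack through constraints with tight swing and twist limits, some locking twist to zero, so the feathers bend only slightly.
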PhysicsAsset /BRCosmetics/Accessories/FORT_Backpacks/Backpack_M_MED_Dark_Minion/Meshes/M_MED_Dark_Minion_Pack_Physics
bodies {
 dyn_feather11_01 [{shapeType=Capsule, length=1.3374, radius=0.75, center=<0.128 0.329 -0.1147>, rotation=<rot 0 81.6278 -80.0746>}]
 dyn_feather11_02 [{shapeType=Capsule, length=0, radius=0.75, center=<0.6239 0.1392 0.0373>, rotation=<rot 0 -99.8821 -65.9216>}]
 dyn_feather12_01 [{shapeType=Capsule, length=1.25, radius=0.5, shapeName=dyn_feather12_01_capsule, center=<1 0 0>, rotation=<rot 90 0 0>}]
 dyn_feather12_02 [{shapeType=Capsule, length=1.0589, radius=0.6324, shapeName=dyn_feather12_02_capsule, center=<1.1 0 0>, rotation=<rot 90 0 0>}]
 dyn_feather12_03 [{shapeType=Capsule, length=0.7788, radius=0.75, shapeName=dyn_feather12_03_capsule, center=<1 0 0>, rotation=<rot 90 0 0>}]
 dyn_feather14_01 [{shapeType=Capsule, length=2.2914, radius=0.75, center=<2 0 0>, rotation=<rot 90 0 0>}]
 dyn_feather14_02 [{shapeType=Capsule, length=2.2623, radius=0.75, center=<2 0 0>, rotation=<rot 90 0 0>}]
 dyn_feather14_03 [{shapeType=Capsule, length=0.5392, radius=0.75, center=<1 0 0>, rotation=<rot 90 0 0>}]
 dyn_feather15_01 [{shapeType=Capsule, length=0.9995, radius=0.75, center=<1.25 0 0>, rotation=<rot 90 0 0>}]
 dyn_feather15_02 [{shapeType=Capsule, length=0, radius=0.75, center=<1.1701 0 0>, rotation=<rot 90 0 0>}]
 dyn_feather17_01 [{shapeType=Sphere, radius=1.2627, shapeName=dyn_feather17_01_sphere, center=<0.3993 0.7969 0.5167>}]
 dyn_feather17_02 [{shapeType=Capsule, length=1.2413, radius=0.75, center=<1.4094 0.0396 0.0428>, rotation=<rot -5.1222 -90 -90.5051>}]
 dyn_feather17_03 [{shapeType=Capsule, length=0.3857, radius=0.75, center=<1.6869 0.0639 -0.2652>, rotation=<rot 0 -93.4023 -69.7327>}]
 dyn_feather19_01 [{shapeType=Capsule, length=1.0993, radius=0.75, center=<1.25 0 0>, rotation=<rot 90 0 0>}]
 dyn_feather19_02 [{shapeType=Capsule, length=1.5506, radius=0.9451, center=<2 0 0>, rotation=<rot 90 0 0>}]
 dyn_feather19_03 [{shapeType=Capsule, length=0.5942, radius=0.75, center=<1 0 0>, rotation=<rot 90 0 0>}]
 dyn_feather1_01 [{shapeType=Capsule, length=0.9006, radius=0.75, center=<1.4123 0.0902 -0.4907>, rotation=<rot 0 -84.0846 -87.628>}]
 dyn_feather1_02 [{shapeType=Capsule, length=1.3236, radius=0.75, center=<1.6555 0 0>, rotation=<rot 90 0 0>}]
 dyn_feather1_03 [{shapeType=Capsule, length=0.0563, radius=0.75, center=<0.746 0 0>, rotation=<rot 90 0 0>}]
 dyn_feather22_01 [{shapeType=Capsule, length=0.9853, radius=0.75, center=<1.25 0 0>, rotation=<rot 90 0 0>}]
 dyn_feather22_02 [{shapeType=Capsule, length=0.8251, radius=0.75, center=<1.25 0 0>, rotation=<rot 90 0 0>}]
 dyn_feather22_03 [{shapeType=Capsule, length=0.8274, radius=0.75, center=<0.9489 0 0>, rotation=<rot 90 0 0>}]
 dyn_feather24_01 [{shapeType=Capsule, length=1, radius=0.5, center=<1.3927 0 0>, rotation=<rot 90 0 0>}]
 dyn_feather24_02 [{shapeType=Capsule, length=1, radius=0.5, center=<1.0838 0 0>, rotation=<rot 90 0 0>}]
 dyn_feather26_01 [{shapeType=Capsule, length=1, radius=0.5, center=<1.1212 0 0>, rotation=<rot 90 0 0>}]
 dyn_feather26_02 [{shapeType=Capsule, length=1, radius=0.5, center=<1.0696 0 0>, rotation=<rot 90 0 0>}]
 dyn_feather26_03 [{shapeType=Capsule, length=1, radius=0.5, center=<1 0 0>, rotation=<rot 90 0 0>}]
 dyn_feather2_01 [{shapeType=Capsule, length=1.5593, radius=0.75, center=<1.2684 0 0>, rotation=<rot 90 0 0>}]
 dyn_feather2_02 [{shapeType=Capsule, length=1.2081, radius=0.75, center=<1.2684 0 0>, rotation=<rot 90 0 0>}]
 dyn_feather2_03 [{shapeType=Capsule, length=1.2441, radius=0.75, center=<1.2684 0 0>, rotation=<rot 90 0 0>}]
 dyn_feather30_01 [{shapeType=Capsule, length=1.25, radius=1, center=<1.5381 0 0>, rotation=<rot 90 0 0>}]
 dyn_feather30_02 [{shapeType=Capsule, length=1.25, radius=1, center=<1.5381 0 0>, rotation=<rot 90 0 0>}]
 dyn_feather32_01 [{shapeType=Capsule, length=1, radius=0.5, center=<1.0429 0 0>, rotation=<rot 90 0 0>}]
 dyn_feather32_02 [{shapeType=Capsule, length=1, radius=0.5, center=<1.108 0 0>, rotation=<rot 90 0 0>}]
 dyn_feather5_01 [{shapeType=Capsule, length=1.3171, radius=0.7983, center=<1.35 0 0>, rotation=<rot 90 0 0>}]
 dyn_feather5_02 [{shapeType=Capsule, length=1.0267, radius=1, center=<1.6407 0 0>, rotation=<rot 90 0 0>}]
 dyn_feather5_03 [{shapeType=Capsule, length=1, radius=1, center=<1.6407 0 0>, rotation=<rot 90 0 0>}]
 dyn_feather7_01 [{shapeType=Capsule, length=1.3062, radius=0.9634, center=<2 0 0>, rotation=<rot 90 0 0>}]
 dyn_feather7_02 [{shapeType=Capsule, length=0.9148, radius=0.75, center=<1.573 0 0>, rotation=<rot 90 0 0>}]
 dyn_feather7_03 [{shapeType=Capsule, length=0.1732, radius=0.75, center=<1.0579 0 0>, rotation=<rot 90 0 0>}]
 dyn_feather9_01 [{shapeType=Capsule, length=1.1105, radius=1, center=<1.6822 0 0>, rotation=<rot 90 0 0>}]
 dyn_feather9_02 [{shapeType=Capsule, length=0.75, radius=1, center=<1.6822 0 0>, rotation=<rot 90 0 0>}]
 dyn_feather9_03 [{shapeType=Capsule, length=0.75, radius=1, center=<1.6822 0 0>, rotation=<rot 90 0 0>}]
 pack [{shapeType=Capsule, length=31.9143, radius=9.5577, center=<0.4241 9.0683 -18.3497>, rotation=<rot -1.3905 -90 -0.9921>}]
}
constraints [
 {bone1Name=dyn_feather1_01, bone2Name=pack, swing1LimitDegrees=10, swing2LimitDegrees=10, twistLimitDegrees=15}
 {bone1Name=dyn_feather1_02, bone2Name=dyn_feather1_01, swing1LimitDegrees=10, swing2LimitDegrees=10, twistLimitDegrees=15}
 {bone1Name=dyn_feather1_03, bone2Name=dyn_feather1_02, swing1LimitDegrees=10, swing2LimitDegrees=10, twistLimitDegrees=15}
 {bone1Name=dyn_feather2_01, bone2Name=pack, swing1LimitDegrees=10, swing2LimitDegrees=10, twistLimitDegrees=15}
 {bone1Name=dyn_feather2_02, bone2Name=dyn_feather2_01, swing1LimitDegrees=10, swing2LimitDegrees=10, twistLimitDegrees=15}
 {bone1Name=dyn_feather2_03, bone2Name=dyn_feather2_02, swing1LimitDegrees=10, swing2LimitDegrees=10, twistLimitDegrees=15}
 {bone1Name=dyn_feather5_01, bone2Name=pack, swing1LimitDegrees=2, swing2LimitDegrees=5, twistLimitDegrees=15}
 {bone1Name=dyn_feather5_02, bone2Name=dyn_feather5_01, swing1LimitDegrees=2, swing2LimitDegrees=5, twistLimitDegrees=5}
 {bone1Name=dyn_feather5_03, bone2Name=dyn_feather5_02, swing1LimitDegrees=5, swing2LimitDegrees=10, twistLimitDegrees=15}
 {bone1Name=dyn_feather7_01, bone2Name=pack, swing1LimitDegrees=2, swing2LimitDegrees=2, twistLimitDegrees=5}
 {bone1Name=dyn_feather7_02, bone2Name=dyn_feather7_01, swing1LimitDegrees=10, swing2LimitDegrees=10, twistLimitDegrees=15}
 {bone1Name=dyn_feather7_03, bone2Name=dyn_feather7_02, swing1LimitDegrees=10, swing2LimitDegrees=10, twistLimitDegrees=15}
 {bone1Name=dyn_feather9_01, bone2Name=pack, swing1LimitDegrees=10, swing2LimitDegrees=10, twistLimitDegrees=15}
 {bone1Name=dyn_feather9_02, bone2Name=dyn_feather9_01, swing1LimitDegrees=10, swing2LimitDegrees=10, twistLimitDegrees=15}
 {bone1Name=dyn_feather9_03, bone2Name=dyn_feather9_02, swing1LimitDegrees=10, swing2LimitDegrees=10, twistLimitDegrees=15}
 {bone1Name=dyn_feather11_01, bone2Name=pack, swing1LimitDegrees=10, swing2LimitDegrees=10, twistLimitDegrees=15}
 {bone1Name=dyn_feather11_02, bone2Name=dyn_feather11_01, swing1LimitDegrees=10, swing2LimitDegrees=10, twistLimitDegrees=15}
 {bone1Name=dyn_feather14_01, bone2Name=pack, swing1LimitDegrees=10, swing2LimitDegrees=10, twistLimitDegrees=0}
 {bone1Name=dyn_feather14_02, bone2Name=dyn_feather14_01, swing1LimitDegrees=10, swing2LimitDegrees=10, twistLimitDegrees=2}
 {bone1Name=dyn_feather14_03, bone2Name=dyn_feather14_02, swing1LimitDegrees=10, swing2LimitDegrees=10, twistLimitDegrees=2}
 {bone1Name=dyn_feather15_01, bone2Name=pack, swing1LimitDegrees=10, swing2LimitDegrees=10, twistLimitDegrees=2}
 {bone1Name=dyn_feather15_02, bone2Name=dyn_feather15_01, swing1LimitDegrees=10, swing2LimitDegrees=10, twistLimitDegrees=15}
 {bone1Name=dyn_feather17_01, bone2Name=pack, swing1LimitDegrees=10, swing2LimitDegrees=10, twistLimitDegrees=0}
 {bone1Name=dyn_feather17_02, bone2Name=dyn_feather17_01, swing1LimitDegrees=10, swing2LimitDegrees=10, twistLimitDegrees=15}
 {bone1Name=dyn_feather17_03, bone2Name=dyn_feather17_02, swing1LimitDegrees=10, swing2LimitDegrees=10, twistLimitDegrees=15}
 {bone1Name=dyn_feather19_01, bone2Name=pack, swing1LimitDegrees=10, swing2LimitDegrees=10, twistLimitDegrees=15}
 {bone1Name=dyn_feather19_02, bone2Name=dyn_feather19_01, swing1LimitDegrees=10, swing2LimitDegrees=10, twistLimitDegrees=15}
 {bone1Name=dyn_feather19_03, bone2Name=dyn_feather19_02, swing1LimitDegrees=10, swing2LimitDegrees=10, twistLimitDegrees=15}
 {bone1Name=dyn_feather22_01, bone2Name=pack, swing1LimitDegrees=10, swing2LimitDegrees=10, twistLimitDegrees=15}
 {bone1Name=dyn_feather22_02, bone2Name=dyn_feather22_01, swing1LimitDegrees=10, swing2LimitDegrees=10, twistLimitDegrees=15}
 {bone1Name=dyn_feather22_03, bone2Name=dyn_feather22_02, swing1LimitDegrees=10, swing2LimitDegrees=10, twistLimitDegrees=15}
 {bone1Name=dyn_feather24_01, bone2Name=pack, swing1LimitDegrees=10, swing2LimitDegrees=10, twistLimitDegrees=15}
 {bone1Name=dyn_feather24_02, bone2Name=dyn_feather24_01, swing1LimitDegrees=10, swing2LimitDegrees=10, twistLimitDegrees=15}
 {bone1Name=dyn_feather26_01, bone2Name=pack, swing1LimitDegrees=10, swing2LimitDegrees=10, twistLimitDegrees=15}
 {bone1Name=dyn_feather26_02, bone2Name=dyn_feather26_01, swing1LimitDegrees=10, swing2LimitDegrees=10, twistLimitDegrees=15}
 {bone1Name=dyn_feather26_03, bone2Name=dyn_feather26_02, swing1LimitDegrees=10, swing2LimitDegrees=10, twistLimitDegrees=15}
 {bone1Name=dyn_feather30_01, bone2Name=pack, swing1LimitDegrees=10, swing2LimitDegrees=10, twistLimitDegrees=15}
 {bone1Name=dyn_feather30_02, bone2Name=dyn_feather30_01, swing1LimitDegrees=10, swing2LimitDegrees=10, twistLimitDegrees=15}
 {bone1Name=dyn_feather32_01, bone2Name=pack, swing1LimitDegrees=10, swing2LimitDegrees=10, twistLimitDegrees=15}
 {bone1Name=dyn_feather32_02, bone2Name=dyn_feather32_01, swing1LimitDegrees=10, swing2LimitDegrees=10, twistLimitDegrees=15}
 {bone1Name=dyn_feather12_01, bone2Name=pack, swing1LimitDegrees=10, swing2LimitDegrees=10, twistLimitDegrees=0}
 {bone1Name=dyn_feather12_02, bone2Name=dyn_feather12_01, swing1LimitDegrees=10, swing2LimitDegrees=10, twistLimitDegrees=0}
 {bone1Name=dyn_feather12_03, bone2Name=dyn_feather12_02, swing1LimitDegrees=10, swing2LimitDegrees=10, twistLimitDegrees=0}
]
