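# Animation Blueprint for the Dark Minion backpack cosmetic, adding secondary motion to the pack and its cape that follows the wearer's spine.
# Each update maps the yaw of the owning mesh's spine_05-to-spine_01 delta from ref pose into pack offset, pack roll and cloth roll.
# From the ref pose, it offsets the pack bone, runs half-weight pack and full-weight cloth RigidBody simulations, then additively rotates three cape bones.
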AnimBlueprint /BRCosmetics/Accessories/FORT_Backpacks/Backpack_M_MED_Dark_Minion/Meshes/M_MED_Dark_Minion_Pack_AnimBP
target_skeleton=&/M_MED_Dark_Minion_Pack_Skeleton
graphs [
 AnimGraph {nodes [
   n1 = "Transform (Modify) Bone - Bone: pack"(ComponentPose=Local To Component(LocalPose=Local Space Ref Pose()), Translation=Get backpack_translation(), Rotation=Get backpack_rotate()) {
    rotationMode=BMM_Additive
    rotationSpace=BCS_BoneSpace
    translationMode=BMM_Additive
    translationSpace=BCS_BoneSpace
    alphaBoolBlend {blendOption=Linear}
    alphaScaleBiasClamp {bInterpResult=true, interpSpeedDecreasing=2, interpSpeedIncreasing=2}
    boneToModify {boneName=pack}
    translation {y=-1, z=-5}
   }
   n2 = RigidBody(ComponentPose=@n1, Alpha=0.5) {
    bClampLinearTranslationLimitToRefPose=true
    bOverrideWorldGravity=true
    overridePhysicsAsset=&/M_MED_Dark_Minion_Pack_Physics
    simulationSpace=BaseBoneSpace
    alphaBoolBlend {blendOption=Linear}
    baseBoneRef {boneName=pack}
    componentAppliedLinearAccClamp=<200 200 200>
    componentLinearAccScale=<0.15 0.3 0.3>
    componentLinearVelScale=<0.05 0.1 0.1>
    overrideWorldGravity {z=-500}
   }
   n3 = RigidBody(ComponentPose=@n2) {
    bClampLinearTranslationLimitToRefPose=true
    bOverrideWorldGravity=true
    overridePhysicsAsset=&/M_MED_Dark_Minion_Cloth_Physics
    simulationSpace=BaseBoneSpace
    alphaBoolBlend {blendOption=Linear}
    baseBoneRef {boneName=pack}
    componentAppliedLinearAccClamp=<2000 2000 1800>
    componentLinearAccScale=<0.3 0.3 0.4>
    componentLinearVelScale=<0.36 0.36 0.6>
    overrideWorldGravity {z=-980}
   }
   n4 = Get backpack_rotate_cloth() -> (backpack_rotate_cloth:Rotator)
   n5 = "Transform (Modify) Bone - Bone: dyn_cape_01"(ComponentPose=@n3, Alpha=1, Translation="0,0,0", Rotation=@n4, Scale="1,1,1") {rotationMode=BMM_Additive, translationMode=BMM_Additive, alphaBoolBlend {blendOption=Linear}, boneToModify {boneName=dyn_cape_01}}
   n6 = "Transform (Modify) Bone - Bone: dyn_cape_01_r"(ComponentPose=@n5, Alpha=1, Translation="0,0,0", Rotation=@n4, Scale="1,1,1") {rotationMode=BMM_Additive, translationMode=BMM_Additive, alphaBoolBlend {blendOption=Linear}, boneToModify {boneName=dyn_cape_01_r}}
   n7 = "Transform (Modify) Bone - Bone: dyn_cape_01_l"(ComponentPose=@n6, Alpha=1, Translation="0,0,0", Rotation=@n4, Scale="1,1,1") {rotationMode=BMM_Additive, translationMode=BMM_Additive, alphaBoolBlend {blendOption=Linear}, boneToModify {boneName=dyn_cape_01_l}}
   Output Pose(Result=Component To Local(ComponentPose=@n7)) {layerGroup=None}
  ]}
 EventGraph {nodes [
   Event BlueprintUpdateAnimation()
   n1 = GetOwnerSkeletalMeshComponent()
   n2 = BreakRotator(InRot=BreakTransform(InTransform=GetDeltaTransformFromRefPose(self=@n1, BoneName=spine_05, BaseName=spine_01)).Rotation)
   > Set backpack_rotate(backpack_rotate=MakeRotator(Roll=MapRangeClamped(Value=@n2.Yaw, InRangeA=0, InRangeB=-30, OutRangeA=8, OutRangeB=30)))
   n3 = MapRangeClamped(Value=@n2.Yaw, InRangeA=0, InRangeB=-25, OutRangeB=75)
   > Set backpack_rotate_cloth(backpack_rotate_cloth:Rotator=MakeRotator(Roll=float + float(A=0, B=@n3) {is=title K2Node_CommutativeAssociativeBinaryOperator})) -> (Output_Get:Rotator)
   > Set backpack_translation(backpack_translation=MakeVector(Y=MapRangeClamped(Value=@n2.Yaw, InRangeA=0, InRangeB=-30, OutRangeB=-5), Z=-5))
   TryGetPawnOwner()
  ]}
]
variables [
 backpack_rotate {type=Rotator}
 backpack_rotate_cloth {type=Rotator}
 backpack_translation {type=Vector}
]
